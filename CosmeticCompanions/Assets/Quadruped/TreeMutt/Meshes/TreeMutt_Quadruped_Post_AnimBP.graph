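# Post-process animation Blueprint for the TreeMutt quadruped cosmetic companion, adding secondary deformation and ribbon motion on top of the input pose.
# A Pose Blend node applies the TreeMutt PoseAsset to the incoming pose.
# Weighted four-target parent constraints then attach the left and right ribbon base bones to neck and spine bones.
# AnimDynamics nodes finally simulate both dynamic ribbon bones in root-relative space with angular springs, limits and custom gravity.
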AnimBlueprint /CosmeticCompanions/Assets/Quadruped/TreeMutt/Meshes/TreeMutt_Quadruped_Post_AnimBP
target_skeleton=&/TreeMutt_Quadruped_Skeleton
graphs [
 AnimGraph {nodes [
   n1 = Input Pose()
   n2 = Get PoseAsset() -> (PoseAsset:Pose Asset Object Reference)
   n3 = "Constraint - base_ribbon_l (PPPP)"!(
    ComponentPose=Local To Component(LocalPose=Pose!(PoseAsset:Pose Asset Object Reference=@n2, SourcePose:Pose Link Structure=@n1) -> (Pose:Pose Link Structure) {is=title AnimGraphNode_PoseBlendNode, blendOption=Linear})
    Alpha=1
    ConstraintWeights_0=0.7377
    ConstraintWeights_1:float=0.1755
    ConstraintWeights_2:float=0.0584
    ConstraintWeights_3:float=0.0284
   ) -> (Pose) {
    boneToModify {boneName=base_ribbon_l}
    constraintSetup [{targetBone {boneName=neck_01}}, {targetBone {boneName=spine_03}}, {targetBone {boneName=neck_02}}, {targetBone {boneName=spine_02}}]
    constraintWeights [1,1,1,1]
   }
   n4 = "Constraint - base_ribbon_r (PPPP)"!(
    ComponentPose=@n3
    Alpha=1
    ConstraintWeights_0=0.7123
    ConstraintWeights_1:float=0.217
    ConstraintWeights_2:float=0.0376
    ConstraintWeights_3:float=0.0332
   ) -> (Pose) {
    boneToModify {boneName=base_ribbon_r}
    constraintSetup [{targetBone {boneName=neck_01}}, {targetBone {boneName=spine_03}}, {targetBone {boneName=spine_02}}, {targetBone {boneName=neck_02}}]
    constraintWeights [1,1,1,1]
   }
   n5 = Anim Dynamics - Bone: dyn_ribbon_r!(ComponentPose=@n4, Alpha=1) -> (Pose) {
    angularDampingOverride=1
    angularSpringConstant=8000
    bAngularSpring=true
    bOverrideAngularDamping=true
    bUseGravityOverride=true
    gravityScale=0.35
    simulationSpace=RootRelative
    alphaScaleBiasClamp {bias=1, scale=-1}
    boundBone {boneName=dyn_ribbon_r}
    componentAppliedLinearAccClamp=<3000 3000 3000>
    componentLinearAccScale=<1 1 1>
    componentLinearVelScale=<1 1 1>
    gravityOverride {z=-980}
    physicsBodyDefinitions [{
      boundBone {boneName=dyn_ribbon_r}
      boxExtents=<30 30 30>
      constraintSetup {angularLimitsMax=<0 15 10>, angularLimitsMin=<0 -15 -10>, angularTarget=<1 0 0>}
      localJointOffset=<5 0 0>
     }]
   }
   n6 = Anim Dynamics - Bone: dyn_ribbon_l!(ComponentPose=@n5, Alpha=1) -> (Pose) {
    angularDampingOverride=1
    angularSpringConstant=8000
    bAngularSpring=true
    bOverrideAngularDamping=true
    bUseGravityOverride=true
    gravityScale=0.35
    simulationSpace=RootRelative
    alphaScaleBiasClamp {bias=1, scale=-1}
    boundBone {boneName=dyn_ribbon_l}
    componentAppliedLinearAccClamp=<3000 3000 3000>
    componentLinearAccScale=<1 1 1>
    componentLinearVelScale=<1 1 1>
    gravityOverride {z=-980}
    physicsBodyDefinitions [{
      boundBone {boneName=dyn_ribbon_l}
      boxExtents=<30 30 30>
      constraintSetup {angularLimitsMax=<0 15 10>, angularLimitsMin=<0 -15 -10>, angularTarget=<1 0 0>}
      localJointOffset=<5 0 0>
     }]
   }
   Output Pose(Result=Component To Local(ComponentPose=@n6))
  ]}
 EventGraph {links [], nodes []}
]
variables [PoseAsset {type=PoseAsset, default=&/TreeMutt_Quadruped_PoseAsset}]
